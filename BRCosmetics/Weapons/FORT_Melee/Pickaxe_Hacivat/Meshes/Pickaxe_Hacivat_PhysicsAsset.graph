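# Physics asset for the Hacivat two-handed melee pickaxe, giving its skeletal mesh simple collision and a jointed link.
# A box body covers the Melee_TwoHanded_Handle bone, and a small capsule covers the dyn_Link1 bone.
# One constraint joins the link to the handle with both swing axes locked, leaving twist as the link's only rotational freedom, like a hinge.
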PhysicsAsset /BRCosmetics/Weapons/FORT_Melee/Pickaxe_Hacivat/Meshes/Pickaxe_Hacivat_PhysicsAsset
bodies {
 Melee_TwoHanded_Handle [{shapeType=Box, extentX=51.4489, extentY=17.118, extentZ=114.79, center=<-0.0234 4.2859 22.3993>, rotation=<rot 0.8046 -90 0.0112>}]
 dyn_Link1 [{shapeType=Capsule, length=6.6985, radius=5.4121, center=<0.1951 0.0388 -5.5403>, rotation=<rot -0.4595 -90 3.0506>}]
}
constraints [{
  bone1Name=dyn_Link1
  bone2Name=Melee_TwoHanded_Handle
  swing1LimitDegrees=45
  swing1Motion=Locked
  swing2LimitDegrees=45
  swing2Motion=Locked
  twistLimitDegrees=89.5714
 }]
